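# MetaSound patch that turns MIDI notes into a float ADSR envelope value.
# A Harmonix MIDI note trigger reads the MIDI stream, with Note On firing the attack and Note Off the release.
# Note velocity divided by 255 sets the sustain level, while attack, decay and release times come from inputs.
# The Track Number input is exposed but not connected to any node.
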MetaSoundPatch /Harmonix/MetaSounds/MSP_MidiNoteAdsr
inputs [
 Attack Time {type=Time}
 Decay Time {type=Time}
 MidiStream {type=MidiStream}
 Release Time {type=Time}
 Track Number {type=Int32}
]
outputs [EnvelopeValue {type=Float}]
pages [Default {
  input_defaults {Attack Time=0.01, Decay Time=0.2, MidiStream=null, Release Time=1, Track Number=1}
  nodes [
   n1 = HarmonixNodes.MidiNoteTrigger(MIDI Stream=^MidiStream)
   ADSREnvelopeADSREnvelope_0 = ADSR Envelope.ADSR Envelope:Float(
    Attack Time=^Attack Time
    Decay Time=^Decay Time
    Release Time=^Release Time
    Sustain Level=UE.Divide:Float(PrimaryOperand=Convert.Float:Int32(Int32=@n1.Velocity).Float) {also {defaults {AdditionalOperands=255}}}.Out
    Trigger Release=@n1.Note Off
    Trigger Attack=@n1.Note On
   )
  ]
  links [{dst="@EnvelopeValue", src=ADSREnvelopeADSREnvelope_0.Out Envelope}]
 }]
